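# UI material for the Nitro Gauntlets reticle, with an IsPlayerInRange parameter that changes its look.
# Material functions build a segmented circle outline, center dot and mirrored triangles, while a symmetry-mirrored SDF texture supplies weapon brackets and their shadow.
# Named reroutes carry these masks into sums and lerps that mix white, black and an in-range orange, with parameters for opacity, shadow opacity and layout.
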Material /NitroGauntletsGameplay/UI/Art/Materials/M_UI_NitroGauntletsReticle
graph {
 properties {
  bAutomaticallySetUsageInEditor=false
  blendMode=BLEND_Translucent
  materialDomain=MD_UI
  pixelDepthOffsetMode=PDOM_Legacy
  usedShadingModels=Default Lit
 }
 nodes [
  n1 = NamedRerouteUsage()
  n2 = ScalarParameter() {parameterName=IsPlayerInRange}
  Add(A=Add(A=NamedRerouteUsage(), B=LinearInterpolate(B=@n1, Alpha=@n2)), B=NamedRerouteUsage())
  n3 = NamedRerouteUsage()
  n4 = NamedRerouteUsage()
  n5 = ScalarParameter() {defaultValue=1, parameterName=Opacity}
  Add(
   A=Add(A=@n3, B=Multiply(A=@n4, B=@n5))
   B=Multiply(A=NamedRerouteUsage(), B=ScalarParameter() {defaultValue=0.7, parameterName=ShadowOpacity})
  )
  LinearInterpolate(A=Constant3Vector(), B=Constant3Vector() {constant=<rgba 1 1 1 0>}, Alpha=Add(A=@n4, B=@n3))
  n6 = Constant3Vector() {constant=<rgba 1 1 1 0>}
  LinearInterpolate(
   A=@n6
   B=LinearInterpolate(A=@n6, B=Constant3Vector() {constant=<rgba 0.9647 0.2747 0 1>}, Alpha=ScalarParameter() {parameterName=IsPlayerInRange})
   Alpha=NamedRerouteUsage()
  )
  n7 = SmoothStep(Value=MaterialFunctionCall(
    Size=ScalarParameter() {defaultValue=0.8, parameterName=CircleSize}
    OutlineThickness=ScalarParameter() {defaultValue=0.03, parameterName=CircleThickness}
   ) {materialFunction=&MF_UI_Circle}.SDFOutline) {constMax=0.01}
  n8 = SmoothStep(Value=MaterialFunctionCall!(UV, Segment Count=Constant() {r=2}, Segment Spacing=ScalarParameter() {defaultValue=0.372, parameterName=SegmentSpacing}, Fixed Spacing) {materialFunction=/FortUILibrary/MaterialFunctions/Basics/MF_UI_RadialSegments}.Segmented Triangles Gradient) {constMax=0.01}
  NamedRerouteDeclaration(None=Multiply(A=OneMinus(None=@n7), B=OneMinus(None=@n8))) {name=Reticle, nodeColor=<rgba 0 0 0 1>}
  NamedRerouteDeclaration(None=MaterialFunctionCall(Size=Constant() {r=0.05}) {materialFunction=&MF_UI_Circle}.Fill) {name=MiddleCircle, nodeColor=<rgba 0 0 0 1>}
  n9 = Add(
   A=MaterialFunctionCall!(UV, Mirror Y, Swizzle UV) {materialFunction=/FortUILibrary/MaterialFunctions/Basics/MF_UI_Symmetry_2x1_New}.Symmetry
   B=AppendVector(A=ScalarParameter() {defaultValue=0.07, parameterName=Distance}, B=Constant())
  )
  n10 = Constant() {r=0.25}
  n11 = MaterialFunctionCall!(
   UVs=MaterialFunctionCall(
    UVs=MaterialFunctionCall(UVs=@n9, "Rotation Angle (0-1)"=@n10) {materialFunction=&CustomRotator}.Rotated Values
    Texture Scale=Constant2Vector() {g=0.2, r=0.1}
   ) {materialFunction=&ScaleUVsByCenter}.UVs
   Corner 01
   Corner 02
   Corner 03
   Glow Min
   Glow Max
   Outline Thickness
  ) {materialFunction=/FortUILibrary/MaterialFunctions/Shapes/MF_UI_WIP_Triangle}
  NamedRerouteDeclaration(None=OneMinus(None=SmoothStep(Value=@n11.SDF) {constMax=0.01})) {name=PlayerInRangeReticles, nodeColor=<rgba 0 0 0 1>}
  n12 = Add(
   A=MaterialFunctionCall!(UV, Swizzle UV) {materialFunction=/FortUILibrary/MaterialFunctions/Basics/MF_UI_Symmetry_2x2_New}.Symmetry
   B=AppendVector(A=ScalarParameter() {defaultValue=-0.05, parameterName=OffsetX}, B=ScalarParameter() {defaultValue=-0.1, parameterName=OffsetY})
  )
  n13 = Constant() {r=0.5}
  n14 = MaterialFunctionCall!(
   RGBA Channels=TextureSample(UVs=MaterialFunctionCall(UVs=@n12, Texture Scale=@n13) {materialFunction=&ScaleUVsByCenter}.UVs) {samplerSource=SSM_Clamp_WorldGroupSettings, texture=&/T_UI_NitroGauntletReticle_SDF}.R
   Glow Min=Constant() {r=0.3}
   Glow Max=Constant() {r=0.7}
   Outline Thickness=Constant() {r=1}
  ) {materialFunction=/FortUILibrary/MaterialFunctions/Basics/MF_UI_ChannelPackedDistanceField}
  NamedRerouteDeclaration(None=@n14.Fill) {name=WeaponBrackets, nodeColor=<rgba 0 0 0 1>}
  NamedRerouteDeclaration(None=Saturate(None=Subtract(A=@n14.DF, B=@n14.DF))) {name=ShadowBracket, nodeColor=<rgba 0 0 0 1>}
 ]
}
