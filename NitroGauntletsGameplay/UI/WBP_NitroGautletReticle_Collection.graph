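# Widget Blueprint for the Nitro Gauntlets reticle's stack pip collection, built on the shared recharge-stack pip collection base.
# A radial Dynamic Entry Box inside a fixed-size Size Box lays out Nitro Gauntlet pip entries along an arc, and Construct hands it to the base.
# PreConstruct and the GenerateStacks override, after calling its parent, loop over entries and rotate each pip by an angle selected from five hardcoded index-based values.
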WidgetBlueprint /NitroGauntletsGameplay/UI/WBP_NitroGautletReticle_Collection
graphs [
 EventGraph {nodes [
   Event PreConstruct()
   > n1 = For Each Loop(Array:Array of User Widget Object References=GetAllEntries!(self:Dynamic Entry Box Base Object Reference=Get DynamicEntryBox_Stacks() -> (DynamicEntryBox_Stacks:Dynamic Entry Box Object Reference)) -> (ReturnValue:Array of User Widget Object References)) -> (Array Element:User Widget Object Reference)
   > n2 = Cast To WBP_NitroGautletReticle_Pip!(execute, Object=@n1) -> (then, CastFailed, AsWBP Nitro Gautlet Reticle Pip:WBP Nitro Gautlet Reticle Pip Object Reference) {is=title K2Node_DynamicCast}
   > RotatePip!(
    execute
    self:WBP Nitro Gautlet Reticle Pip Object Reference=@n2
    Transform Angle:float=Select!(Option 0:float=60, Option 1:float=40, Option 2:float=24, Option 3:float=4, Option 4:float=-12, Index:int=@n1.Array Index) -> (ReturnValue:float)
   ) -> (then)
   Event Construct()
   > SetDynamicEntryBox!(
    execute
    self:Fort Weapon Reticle Recharge Stack Base Pip Collection Base Object Reference
    DynamicEntryBox:Dynamic Entry Box Object Reference=Get DynamicEntryBox_Stacks()
   ) -> (then)
  ]}
 GenerateStacks {nodes [
   n1 = GenerateStacks!() -> (then, TotalStacks:int=0, CurrentStacks:int=0) {is=title K2Node_FunctionEntry}
   > Parent: GenerateStacks!(execute, TotalStacks:int=@n1, CurrentStacks:int=@n1.CurrentStacks) -> (then) {is=title K2Node_CallParentFunction}
   > n2 = For Each Loop(Array=GetAllEntries(self=Get DynamicEntryBox_Stacks()))
   > n3 = Cast To WBP_NitroGautletReticle_Pip(Object=@n2)
   > RotatePip(self=@n3, Transform Angle=Select(Option 0=61, Option 1=40, Option 2=24, Option 3=4, Option 4=-12, Index=@n2.Array Index))
  ]}
]
variables [
 AngleT {type=number}
 NitroGautletRef {type=B_Weap_Moonflax_NitroGauntlets_C, default=null}
]
widgets {
 info {
  inheritedWidgetCount=0
  parentClass=/Game/Athena/HUD/reticles/ReticleExtensionWidgets/ChargeStack/WBP_Reticle_RechargeStackBase_PipCollection.WBP_Reticle_RechargeStackBase_PipCollection_C
  rootWidgetClass=&SizeBox
  widgetCount=2
 }
 widgets [
  SizeBox_0 {
   widgetClassPath=&SizeBox
   widget=SizeBox_0
   properties {bOverride_HeightOverride=true, bOverride_WidthOverride=true, heightOverride=190, widthOverride=190}
  }
  DynamicEntryBox_Stacks {
   widgetClassPath=&DynamicEntryBox
   bIsVariable=true
   parent=SizeBox_0
   widget=DynamicEntryBox_Stacks
   properties {
    entryBoxType=Radial
    entryWidgetClass=&/WBP_NitroGautletReticle_Pip.WBP_NitroGautletReticle_Pip_C
    numDesignerPreviewEntries=5
    radialBoxSettings {angleBetweenItems=18, bClockwiseOrder=false, bDistributeItemsEvenly=false, marginSize=0, sectorCentralAngle=360, startingAngle=301}
   }
   slot {class=SizeBoxSlot}
  }
 ]
}
