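# Reticle widget for the Nitro Gauntlets weapon, extending the recharge-stack reticle base with a pip collection and a material-driven reticle image.
# Construct plays an intro animation; it and ammo changes toggle a polling timer that runs while the magazine is below full.
# Cooldown updates read the weapon item's remaining time into the pips' Progress material parameter.
# It binds to the weapon's damage event, and SetDamageDeltAnimation pulses reticle material offsets for non-environment hits.
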WidgetBlueprint /NitroGauntletsGameplay/UI/WBP_NitroGautletReticle
animations [
 Intro {
  binding_count=1
  length_seconds=0.5
  object=&self:Intro
  sequence {bindings [Root {
     id=#1
     possessed_class=Overlay
     tracks [
      {
       type=MovieScene2DTransformTrack
       sections [{
         type=MovieScene2DTransformSection
         end_frame=10
         start_frame=0
         channels [
          Scale.X {type=MovieSceneScriptingFloatChannel, keys [[0,1.25],[10,1]]}
          Scale.Y {type=MovieSceneScriptingFloatChannel, keys [[0,1.25],[10,1]]}
         ]
        }]
      }
      {
       type=MovieSceneFloatTrack
       display_name=Render Opacity
       sections [{type=MovieSceneFloatSection, end_frame=10, start_frame=0, channels [None {type=MovieSceneScriptingFloatChannel, keys [[0,0],[10,1]]}]}]
      }
     ]
    }]}
 }
 Anim_Attack {
  binding_count=1
  length_seconds=0.2
  object=&self:Anim_Attack
  sequence {bindings [Reticle {
     id=#2
     possessed_class=Image
     tracks [{
       type=MovieSceneWidgetMaterialTrack
       sections [{
         type=MovieSceneParameterSection
         end_frame=4
         start_frame=0
         channels [
          OffsetX {type=MovieSceneScriptingFloatChannel, keys [[0,-0.05],[2,-0.23],[4,-0.05]]}
          OffsetY {type=MovieSceneScriptingFloatChannel, keys [[0,-0.1],[2,-0.25],[4,-0.1]]}
          Opacity {type=MovieSceneScriptingFloatChannel, const [1,2]}
         ]
        }]
      }]
    }]}
 }
]
graphs [
 BindToWeaponEvents {nodes [
   n1 = BindToWeaponEvents!() -> (then, DoBind:bool, EquippedWeapon:Fort Weapon Object Reference) {is=title K2Node_FunctionEntry}
   > Parent: BindToWeaponEvents!(execute, DoBind:bool=@n1, EquippedWeapon:Fort Weapon Object Reference=@n1.EquippedWeapon) -> (then) {is=title K2Node_CallParentFunction}
   > n2 = Cast To B_Weap_Moonflax_NitroGauntlets!(execute, Object=@n1.EquippedWeapon) -> (then, CastFailed, AsB Weap Moonflax Nitro Gauntlets:B Weap Moonflax Nitro Gauntlets Object Reference) {is=title K2Node_DynamicCast}
   > n3 = Set NitroGautletRef(NitroGautletRef:B Weap Moonflax Nitro Gauntlets Object Reference=@n2) -> (Output_Get:B Weap Moonflax Nitro Gauntlets Object Reference)
   > Bind Event to On Did Damage(self:B Weap Moonflax Nitro Gauntlets Object Reference=@n3, Delegate=Create Event()) {is=title K2Node_AddDelegate}
  ]}
 EventGraph {nodes [
   n1 = Event OnLocalAmmoChanged!() -> (OutputDelegate, then, LocalCount:int=0, LocalRemaining:int=0) {is=title K2Node_Event}
   > Parent: OnLocalAmmoChanged!(execute, LocalCount:int=@n1.LocalCount, LocalRemaining:int=@n1.LocalRemaining) -> (then) {is=title K2Node_CallParentFunction}
   > UpdateCooldownQueryTimer(self:Fort Weapon Reticle Extension Widget Base Object Reference)
   Event Construct()
   > Parent: Construct()
   > PlayAnimation(InAnimation=Get Intro())
   > UpdateCooldownQueryTimer()
   Event Destruct()
   > Parent: Destruct()
   > HandleCooldownTimer!(execute, self:Fort Weapon Reticle Extension Widget Base Object Reference, bActivateTimer:bool=false, TimerInterval:float=0) -> (then)
   Event UpdateRemainingCooldown() {is=title K2Node_Event}
   > n2 = GetRemainingCooldownTimer!(execute, self) -> (then, RemainingTime:double, TotalDuration:double)
   > UpdateRemainingRechargeTime!(
    execute
    self:Fort Weapon Reticle Recharge Stack Base Pip Collection Base Object Reference=GetPipCollection!(self:WBP Reticle Recharge Stack Base Object Reference) -> (PipCollection:WBP Reticle Recharge Stack Base Pip Collection Object Reference)
    TimeRemaining:float=@n2
    TotalDuration:float=@n2.TotalDuration
    PipPercentMatParamName:Name=Progress
    ShouldSetPipColor:bool=false
    NotReadyColor:Linear Color Structure
    ReadyColor:Linear Color Structure
    PipColorMatParamName:Name=None
   ) -> (then)
   Event UpdateCooldownQueryTimer() {is=title K2Node_Event}
   > n3 = Get!(execute:Exec) -> (then:Exec, else:Exec, NitroGautletRef:B Weap Moonflax Nitro Gauntlets Object Reference)
   > n4 = Branch(Condition="integer >= integer"(
     A=GetLocalMagazineAmmoCount!(self:Fort Weapon Object Reference=@n3) -> (ReturnValue:int=0)
     B=GetBulletsPerClip!(self:Fort Weapon Object Reference=@n3) -> (ReturnValue:int=0)
    ))
   > HandleCooldownTimer(bActivateTimer=false, TimerInterval=0)
   HandleCooldownTimer(execute=@n4.else, bActivateTimer=true, TimerInterval=0.1)
  ]}
 GetPipCollection {nodes [
   GetPipCollection() {is=title K2Node_FunctionEntry}
   > Return Node!(
    execute
    PipCollection:WBP Reticle Recharge Stack Base Pip Collection Object Reference=Get WBP_NitroGautletReticle_Collection() -> (WBP_NitroGautletReticle_Collection:WBP Nitro Gautlet Reticle Collection Object Reference)
   )
  ]}
 GetRemainingCooldownTimer {nodes [
   GetRemainingCooldownTimer() {is=title K2Node_FunctionEntry}
   > n1 = Cast To FortPlayerController(Object=GetOwningPlayer())
   > n2 = Get()
   n3 = GetCoolDownTimeRemaining!(
    self:Fort Item Object Reference=GetInventoryItem!(self:Fort Weapon Object Reference=@n2) -> (ReturnValue:Fort Item Object Reference)
    FortPC:Fort Player Controller Object Reference=@n1
   ) -> (CooldownDuration:float, ReturnValue:float)
   > Return Node!(execute, RemainingTime:double=@n3.ReturnValue, TotalDuration:double=@n3)
   Return Node(execute=[@n2.else, @n1.CastFailed], RemainingTime=0, TotalDuration=0)
  ]}
 SetDamageDeltAnimation {nodes [
   n1 = SetDamageDeltAnimation!() -> (then, Damage:double=~, EnviroHit?:bool=~) {is=title K2Node_FunctionEntry}
   > n2 = Branch(Condition=@n1.EnviroHit?)
   PlayAnimationForward(execute=@n2.else, InAnimation=Get Anim_Attack() -> (Anim_Attack:Widget Animation Object Reference))
  ]}
]
variables [
 FoundTarget {type=integer}
 NitroGautletRef {type=B_Weap_Moonflax_NitroGauntlets_C, default=null}
 NoTarget {type=integer}
]
widgets {
 info {
  inheritedWidgetCount=1
  parentClass=/Game/Athena/HUD/reticles/ReticleExtensionWidgets/ChargeStack/WBP_Reticle_RechargeStackBase.WBP_Reticle_RechargeStackBase_C
  rootWidgetClass=&Overlay
  widgetCount=4
 }
 widgets [
  Root {widgetClassPath=&Overlay, bIsVariable=true, widget=Root}
  WBP_NitroGautletReticle_Collection {
   widgetClassPath=&/WBP_NitroGautletReticle_Collection
   bIsVariable=true
   parent=Root
   widget=WBP_NitroGautletReticle_Collection
   properties {renderTransform {angle=4, scale=<1 1>, shear=<0 0>, translation=<0 0>}}
   slot {class=OverlaySlot}
  }
  Reticle {
   widgetClassPath=&Image
   bIsVariable=true
   parent=Root
   widget=Reticle
   properties {brush {
     bIsDynamicallyLoaded=false
     drawAs=Image
     imageType=FullColor
     mirroring=NoMirror
     resourceName=None
     resourceObject=/NitroGauntletsGameplay/UI/Art/Materials/M_UI_NitroGauntletsReticle
     tiling=NoTile
     imageSize=<80 80>
     margin=<ltrb 0 0 0 0>
     outlineSettings {
      bUseBrushTransparency=false
      roundingType=HalfHeightRadius
      width=0
      color {colorUseRule=UseColor_Specified, specifiedColor=<rgba 0 0 0 0>}
      cornerRadii=<0 0 0 0>
     }
     tintColor {colorUseRule=UseColor_Specified, specifiedColor=<rgba 1 1 1 1>}
     uVRegion {bIsValid=false, max=<0 0>, min=<0 0>}
    }}
   slot {class=OverlaySlot, properties {horizontalAlignment=HAlign_Center, verticalAlignment=VAlign_Center}}
  }
  TutorialHighlight {widgetClassPath=&TutorialHighlight}
 ]
}
